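# Niagara effect for the BountyPond emote bucket, in a DeleteSoon folder, continuously spawning short-lived mesh particles tinted by time of day.
# One emitter spawns them on a squashed spherical shape, with front-end modifiers on placement and camera offset.
# A stepped distribution curve makes roughly half stationary and launches the rest.
# Initial speed drives two lerps: launched particles fall under scaled gravity, while stationary ones follow a randomized up-then-down velocity curve.
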NiagaraSystem /BRCosmetics/Animation/Game/MainPlayer/Emotes/BountyPond/FX/DeleteSoon/NS_Emote_BountyPond_Bucket1
dependencies {
 usedDataInterfaces [&NiagaraDataInterfaceCurve, &NiagaraDataInterfaceVectorCurve]
 usedDynamicInputs []
 usedModules [&SystemState]
}
emitters {Minimal001 {
  emitterUpdateScript [
   &EmitterState()
   &SpawnRate(SpawnRate=100, Use Spawn Probability=false)
  ]
  particleSpawnScript [
   &InitializeParticle(
    Color=&FNTimeOfDayColor(
     Night Color=<rgba 0.2292 0.4152 1 1>
     Morning Color=<rgba 1 0.6618 0.5365 1>
     Day Color=<rgba 1 1 1 1>
     Evening Color=<rgba 1 0.6618 0.5365 1>
    )
    Lifetime Max=0.7
    Lifetime Min=0.5
    Lifetime Mode=::NewEnumerator1
    Mesh Scale Mode=::NewEnumerator3
    Mesh Uniform Scale=&Multiply_Float(A=&RandomRangeFloat(Minimum=0.35, Maximum=0.25), B=0.75)
    Sprite Size Mode=::NewEnumerator3
    Uniform Sprite Size=7
    UsePositionOffset=false
    Write Material Random=true
   )
   &ShapeLocation(
    Hemisphere Distribution=<1 0.5>
    MaskSpawnGroup=false
    Non Uniform Scale=&FN_FrontEndVector3Modifier(Gameplay=<1.75 1.75 0.6>, Front End=<1.75 1.75 0.6>)
    Offset=<0 0 -5>
    Offset Mode=::NewEnumerator0
    Override Seed=false
    Sphere Radius=&FN_FrontEndFloatModifier(Gameplay=5, Front End=5)
    Sphere Surface Distribution=1
    UseOwnerScale=true
   )
   &CameraOffset(Camera Offset Amount=3)
   &AddVelocity(
    Module.Use Falloff Distance=false
    Origin Offset=<0 0 -15>
    Override Seed=false
    Velocity Mode=::NewEnumerator1
    Velocity Speed=/Niagara/DynamicInputs/Helpers/DistributionCurveFloat(
     Minimum:NiagaraFloat=0
     Maximum:NiagaraFloat=60
     Distribution Curve:NiagaraDataInterfaceCurve=DataInterface() {propertyValues {Curve {
        defaultValue=0
        keys [{tangentMode=RCTM_User, time=0, value=0}, <at 0.49 0 RCIM_Constant>, <at 0.5014 1 RCIM_Constant>, {tangentMode=RCTM_User, time=1, value=1}]
       }}}
     Randomness Mode:ENiagaraRandomnessMode=::NewEnumerator0
     Evaluation Type:ENiagaraRandomnessEvaluation=::NewEnumerator0
     Seed:NiagaraInt32=0
     Fixed Random Seed:NiagaraBool=false
    )
   )
   &ApplyInitialForces()
   &InitialMeshOrientation(Use Rotation Vector=false)
   &CameraOffset(Camera Offset Amount=&FN_FrontEndFloatModifier(Gameplay=0, Front End=-2)) {module=CameraOffset001}
  ]
  particleUpdateScript [
   &ParticleState()
   &UpdateMeshOrientation()
   &GravityForce(Gravity=&Lerp_Vector(
     A=<0 0 0>
     B=<0 0 -980>
     Alpha=&RemapRange(
      Input Value To Remap=&VectorLength(Vector=$Particles.Initial.Velocity, Safe Length Calculation=false)
      Input Min=0
      Input Max=100
      Output Min=0
      Output Max=1
     )
    ))
   &SolveForcesAndVelocity(Clamp Velocity=false, Limit Acceleration=false)
   /SetVariables(Particles.Velocity=&Lerp_Vector(
     A=&VectorFromCurve(
      VectorCurve=DataInterface() {propertyValues {XCurve {keys [<at 0 1>,<at 1 1>]}, YCurve {keys [<at 0 1>,<at 1 1>]}, ZCurve {keys [<at 0 0>,<at 0.2 1>,<at 0.3 -1>,<at 0.5 0>]}}}
      Scale Curve=&RandomRangeVector(Minimum=<-2 -2 15>, Maximum=<2 2 15>)
     )
     B=$Particles.Velocity
     Alpha=&RemapRange(
      Input Value To Remap=&VectorLength(Vector=$Particles.Initial.Velocity, Safe Length Calculation=false)
      Input Min=0
      Input Max=1
      Output Min=0
      Output Max=1
     )
    )) {isSetParameters=true}
  ]
  rendererClasses [&NiagaraMeshRendererProperties]
  renderers [{rendererClass=&NiagaraMeshRendererProperties, rendererIndex=0}]
 }}
summary {
 systemName=NS_Emote_BountyPond_Bucket1
 emitters [{emitterName=Minimal001, rendererClasses [&NiagaraMeshRendererProperties]}]
 userVariables [
  User.bIsFrontEnd {type=NiagaraBool}
  User.bIsFrontEndPreview {type=NiagaraBool}
  User.NormalizedNotifyProgress {type=NiagaraFloat, defaultValue=0, description=""}
 ]
}
