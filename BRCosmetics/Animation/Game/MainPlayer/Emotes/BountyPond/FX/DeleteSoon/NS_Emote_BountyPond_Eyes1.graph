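# Niagara effect for the BountyPond emote that spews popcorn mesh particles from a character's left and right eyes, stored in a DeleteSoon folder.
# Two mirrored Flow emitters spawn at head-bone offsets, pop in with a scale overshoot, and arc under half gravity with time-of-day tint.
# User switches index arrays: EyeOffsetSwitch picks eye offset and lifetime, CharacterSwitch picks launch velocity.
# Two short-lived, stationary Still emitters are disabled.
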NiagaraSystem /BRCosmetics/Animation/Game/MainPlayer/Emotes/BountyPond/FX/DeleteSoon/NS_Emote_BountyPond_Eyes1
dependencies {
 usedDataInterfaces [
  &NiagaraDataInterfaceSkeletalMesh
  &NiagaraDataInterfaceArrayFloat3
  &NiagaraDataInterfaceCurve
  &NiagaraDataInterfaceArrayFloat
 ]
 usedDynamicInputs []
 usedModules [
  &self:SystemSpawnScript.SystemScriptSource.SystemScriptGraph.NiagaraNodeAssignment_3.SetVariables_#1
  &SystemState
  &BlendSpaceHelper
 ]
}
emitters {
 PopCornFlow_L {
  emitterUpdateScript [
   &EmitterState()
   &SpawnRate(SpawnRate=9, Use Spawn Probability=false)
  ]
  particleSpawnScript [
   &InitializeParticle(
    Color=&FNTimeOfDayColor(
     Night Color=<rgba 0.2292 0.4152 1 1>
     Morning Color=<rgba 1 0.6618 0.5365 1>
     Day Color=<rgba 0.9739 1.0795 1.15 1>
     Evening Color=<rgba 1 0.6618 0.5365 1>
    )
    Lifetime=&SelectFloatFromArray(
     Direct Array Index=&MakeIntFromFloat(Float=$User.EyeOffsetSwitch)
     Float Selection Array=$System.PopcornLifetime
    )
    Mesh Scale Mode=::NewEnumerator3
    Mesh Uniform Scale=&Multiply_Float(A=&RandomRangeFloat(Minimum=0.35, Maximum=0.25), B=0.95)
    Position Offset Coordinate Space=::Local
    Sprite Rotation Mode=::NewEnumerator2
    Sprite Size Mode=::NewEnumerator3
    Uniform Sprite Size=7
    UsePositionOffset=true
    Write Material Random=true
   )
   &SkeletalMeshLocation(
    Bone Sampling Mode=::NewEnumerator4
    MaskSpawnGroup=false
    Sampled Position Offset=&SelectVectorFromArray(
     Direct Array Index=&MakeIntFromFloat(Float=$User.EyeOffsetSwitch)
     Vector Selection Array=$System.Vector Array L
    )
    Sampled Position Offset Coordinate Space:ENiagaraSkelMeshTransforms=::NewEnumerator2
    Skeletal Mesh=DataInterface() {propertyValues {
      DefaultMesh=null
      ExcludeBoneName=root
      OptionalComponentClass=null
      PreviewMesh=null
      SkinningMode=SkinOnTheFly
      SoftSourceActor=null
      SourceMode=Default
      UvSetIndex=0
      WholeMeshLOD=-1
      bExcludeBone=false
      bReadDeformedGeometry=true
      bRequireCurrentFrameData=true
      ComponentTags []
      FilteredBones [head]
      FilteredSockets []
      MeshUserParameter {parameter {name=None, typeDefHandle {registeredTypeIndex=1}, varData []}}
      SamplingRegions []
     }}
   )
   &ShapeLocation(MaskSpawnGroup=false, Override Seed=false, Sphere Radius=2, UseOwnerScale=true)
   &InitialMeshOrientation(Use Rotation Vector=false)
   &CameraOffset(Camera Offset Amount=&FN_FrontEndFloatModifier(Gameplay=0, Front End=0)) {module=CameraOffset001}
   &AddVelocity(Velocity=&SelectVectorFromArray(
     Direct Array Index=&MakeIntFromFloat(Float=$User.CharacterSwitch)
     Vector Selection Array=DataInterface() {propertyValues {FloatData [<0 40 30>, <0 50 50>]}}
    ))
  ]
  particleUpdateScript [
   &ParticleState()
   &ScaleMeshSize(Scale Factor=&Multiply_VectorByFloat(
     Vector=<1 1 1>
     Float=&FloatFromCurve(FloatCurve=DataInterface() {propertyValues {Curve {keys [<at 0 0>,<at 0.0499 1.5>,<at 0.1 1>,<at 1 1>]}}})
    ))
   &UpdateMeshOrientation()
   &SpriteRotationRate(Rotation Rate=300)
   &GravityForce(Gravity=&Multiply_Vector(A=<0 0 -980>, B=<0 0 0.5>))
   &SolveForcesAndVelocity(Clamp Velocity=false, Limit Acceleration=false)
  ]
  rendererClasses [&NiagaraMeshRendererProperties]
  renderers [{rendererClass=&NiagaraMeshRendererProperties, rendererIndex=0}]
 }
 PopCornFlow_R {
  emitterUpdateScript [
   &EmitterState()
   &SpawnRate(SpawnRate=9, Use Spawn Probability=false)
  ]
  particleSpawnScript [
   &InitializeParticle(
    Color=&FNTimeOfDayColor(
     Night Color=<rgba 0.2292 0.4152 1 1>
     Morning Color=<rgba 1 0.6618 0.5365 1>
     Day Color=<rgba 0.9739 1.0795 1.15 1>
     Evening Color=<rgba 1 0.6618 0.5365 1>
    )
    Lifetime=&SelectFloatFromArray(
     Direct Array Index=&MakeIntFromFloat(Float=$User.EyeOffsetSwitch)
     Float Selection Array=$System.PopcornLifetime
    )
    Mesh Scale Mode=::NewEnumerator3
    Mesh Uniform Scale=&Multiply_Float(A=&RandomRangeFloat(Minimum=0.35, Maximum=0.25), B=0.95)
    Position Offset=<-4 0 0>
    Position Offset Coordinate Space=::Local
    Sprite Rotation Mode=::NewEnumerator2
    Sprite Size Mode=::NewEnumerator3
    Uniform Sprite Size=7
    UsePositionOffset=true
    Write Material Random=true
   )
   &SkeletalMeshLocation(
    Bone Sampling Mode=::NewEnumerator4
    MaskSpawnGroup=false
    Sampled Position Offset=&SelectVectorFromArray(
     Direct Array Index=&MakeIntFromFloat(Float=$User.EyeOffsetSwitch)
     Vector Selection Array=$System.Vector Array R
    )
    Sampled Position Offset Coordinate Space:ENiagaraSkelMeshTransforms=::NewEnumerator2
    Skeletal Mesh=DataInterface() {propertyValues {
      DefaultMesh=null
      ExcludeBoneName=root
      OptionalComponentClass=null
      PreviewMesh=null
      SkinningMode=SkinOnTheFly
      SoftSourceActor=null
      SourceMode=Default
      UvSetIndex=0
      WholeMeshLOD=-1
      bExcludeBone=false
      bReadDeformedGeometry=true
      bRequireCurrentFrameData=true
      ComponentTags []
      FilteredBones [head]
      FilteredSockets []
      MeshUserParameter {parameter {name=None, typeDefHandle {registeredTypeIndex=1}, varData []}}
      SamplingRegions []
     }}
   )
   &ShapeLocation(MaskSpawnGroup=false, Override Seed=false, Sphere Radius=2, UseOwnerScale=true)
   &InitialMeshOrientation(Use Rotation Vector=false)
   &CameraOffset(Camera Offset Amount=&FN_FrontEndFloatModifier(Gameplay=0, Front End=0)) {module=CameraOffset001}
   &AddVelocity(Velocity=&SelectVectorFromArray(
     Direct Array Index=&MakeIntFromFloat(Float=$User.CharacterSwitch)
     Vector Selection Array=DataInterface() {propertyValues {FloatData [<0 40 30>, <0 50 50>]}}
    ))
  ]
  particleUpdateScript [
   &ParticleState()
   &ScaleMeshSize(Scale Factor=&Multiply_VectorByFloat(
     Vector=<1 1 1>
     Float=&FloatFromCurve(FloatCurve=DataInterface() {propertyValues {Curve {keys [<at 0 0>,<at 0.0499 1.5>,<at 0.1 1>,<at 1 1>]}}})
    ))
   &UpdateMeshOrientation()
   &SpriteRotationRate(Rotation Rate=300)
   &GravityForce(Gravity=&Multiply_Vector(A=<0 0 -980>, B=<0 0 0.5>))
   &SolveForcesAndVelocity(Clamp Velocity=false, Limit Acceleration=false)
  ]
  rendererClasses [&NiagaraMeshRendererProperties]
  renderers [{rendererClass=&NiagaraMeshRendererProperties, rendererIndex=0}]
 }
 PopCornStill_L {
  bEnabled=false
  emitterUpdateScript [
   &EmitterState()
   &SpawnRate(SpawnRate=7, Use Spawn Probability=false)
  ]
  particleSpawnScript [
   &InitializeParticle(
    Color=&FNTimeOfDayColor(
     Night Color=<rgba 0.2292 0.4152 1 1>
     Morning Color=<rgba 1 0.6618 0.5365 1>
     Day Color=<rgba 0.9739 1.0795 1.15 1>
     Evening Color=<rgba 1 0.6618 0.5365 1>
    )
    Lifetime=0.0333
    Mesh Scale Mode=::NewEnumerator3
    Mesh Uniform Scale=0.15
    Position Offset Coordinate Space=::Local
    Sprite Rotation Mode=::NewEnumerator2
    Sprite Size Mode=::NewEnumerator3
    Uniform Sprite Size=7
    UsePositionOffset=true
    Write Material Random=true
   )
   &SkeletalMeshLocation(
    Bone Sampling Mode=::NewEnumerator4
    MaskSpawnGroup=false
    Sampled Position Offset=&SelectVectorFromArray(
     Direct Array Index=&MakeIntFromFloat(Float=$User.EyeOffsetSwitch)
     Vector Selection Array=$System.Vector Array L
    )
    Sampled Position Offset Coordinate Space:ENiagaraSkelMeshTransforms=::NewEnumerator2
    Skeletal Mesh=DataInterface() {propertyValues {
      DefaultMesh=null
      ExcludeBoneName=root
      OptionalComponentClass=null
      PreviewMesh=null
      SkinningMode=SkinOnTheFly
      SoftSourceActor=null
      SourceMode=Default
      UvSetIndex=0
      WholeMeshLOD=-1
      bExcludeBone=false
      bReadDeformedGeometry=true
      bRequireCurrentFrameData=true
      ComponentTags []
      FilteredBones [head]
      FilteredSockets []
      MeshUserParameter {parameter {name=None, typeDefHandle {registeredTypeIndex=1}, varData []}}
      SamplingRegions []
     }}
   )
   &InitialMeshOrientation(Use Rotation Vector=false)
   &CameraOffset(Camera Offset Amount=&FN_FrontEndFloatModifier(Gameplay=0, Front End=0)) {module=CameraOffset001}
  ]
  particleUpdateScript [
   &ParticleState()
   &ScaleMeshSize(Scale Factor=&Multiply_VectorByFloat(
     Vector=<1 1 1>
     Float=&FloatFromCurve(FloatCurve=DataInterface() {propertyValues {Curve {keys [<at 0 0>,<at 0.0499 1.5>,<at 0.1 1>,<at 1 1>]}}})
    ))
   &UpdateMeshOrientation()
   &SpriteRotationRate(Rotation Rate=300)
   &SolveForcesAndVelocity(Clamp Velocity=false, Limit Acceleration=false)
  ]
  rendererClasses [&NiagaraMeshRendererProperties]
  renderers [{rendererClass=&NiagaraMeshRendererProperties, rendererIndex=0}]
 }
 PopCornStill_R {
  bEnabled=false
  emitterUpdateScript [
   &EmitterState()
   &SpawnRate(SpawnRate=7, Use Spawn Probability=false)
  ]
  particleSpawnScript [
   &InitializeParticle(
    Color=&FNTimeOfDayColor(
     Night Color=<rgba 0.2292 0.4152 1 1>
     Morning Color=<rgba 1 0.6618 0.5365 1>
     Day Color=<rgba 0.9739 1.0795 1.15 1>
     Evening Color=<rgba 1 0.6618 0.5365 1>
    )
    Lifetime=0.0333
    Mesh Scale Mode=::NewEnumerator3
    Mesh Uniform Scale=0.15
    Position Offset Coordinate Space=::Local
    Sprite Rotation Mode=::NewEnumerator2
    Sprite Size Mode=::NewEnumerator3
    Uniform Sprite Size=7
    UsePositionOffset=true
    Write Material Random=true
   )
   &SkeletalMeshLocation(
    Bone Sampling Mode=::NewEnumerator4
    MaskSpawnGroup=false
    Sampled Position Offset=&SelectVectorFromArray(
     Direct Array Index=&MakeIntFromFloat(Float=$User.EyeOffsetSwitch)
     Vector Selection Array=$System.Vector Array R
    )
    Sampled Position Offset Coordinate Space:ENiagaraSkelMeshTransforms=::NewEnumerator2
    Skeletal Mesh=DataInterface() {propertyValues {
      DefaultMesh=null
      ExcludeBoneName=root
      OptionalComponentClass=null
      PreviewMesh=null
      SkinningMode=SkinOnTheFly
      SoftSourceActor=null
      SourceMode=Default
      UvSetIndex=0
      WholeMeshLOD=-1
      bExcludeBone=false
      bReadDeformedGeometry=true
      bRequireCurrentFrameData=true
      ComponentTags []
      FilteredBones [head]
      FilteredSockets []
      MeshUserParameter {parameter {name=None, typeDefHandle {registeredTypeIndex=1}, varData []}}
      SamplingRegions []
     }}
   )
   &InitialMeshOrientation(Use Rotation Vector=false)
   &CameraOffset(Camera Offset Amount=&FN_FrontEndFloatModifier(Gameplay=0, Front End=0)) {module=CameraOffset001}
  ]
  particleUpdateScript [
   &ParticleState()
   &ScaleMeshSize(Scale Factor=&Multiply_VectorByFloat(
     Vector=<1 1 1>
     Float=&FloatFromCurve(FloatCurve=DataInterface() {propertyValues {Curve {keys [<at 0 0>,<at 0.0499 1.5>,<at 0.1 1>,<at 1 1>]}}})
    ))
   &UpdateMeshOrientation()
   &SpriteRotationRate(Rotation Rate=300)
   &SolveForcesAndVelocity(Clamp Velocity=false, Limit Acceleration=false)
  ]
  rendererClasses [&NiagaraMeshRendererProperties]
  renderers [{rendererClass=&NiagaraMeshRendererProperties, rendererIndex=0}]
 }
}
summary {
 systemName=NS_Emote_BountyPond_Eyes1
 emitters [
  {emitterName=PopCornFlow_L, rendererClasses [&NiagaraMeshRendererProperties]}
  {emitterName=PopCornFlow_R, rendererClasses [&NiagaraMeshRendererProperties]}
  {bEnabled=false, emitterName=PopCornStill_L, rendererClasses [&NiagaraMeshRendererProperties]}
  {bEnabled=false, emitterName=PopCornStill_R, rendererClasses [&NiagaraMeshRendererProperties]}
 ]
 userVariables [
  User.bIsFrontEnd {type=NiagaraBool}
  User.bIsFrontEndPreview {type=NiagaraBool}
  User.CharacterSwitch {type=NiagaraFloat, defaultValue=0, description=""}
  User.EyeOffset {type=NiagaraFloat, defaultValue=0, description=""}
  User.EyeOffsetSwitch {type=NiagaraFloat, defaultValue=0, description=""}
  User.LifeTimeSwitch {type=NiagaraFloat, defaultValue=0, description=""}
  User.NormalizedNotifyProgress {type=NiagaraFloat, defaultValue=0, description=""}
 ]
}
